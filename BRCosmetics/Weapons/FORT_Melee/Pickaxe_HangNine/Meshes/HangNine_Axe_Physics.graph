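# Physics asset for the HangNine pickaxe mesh that sets up physical motion for a necklace of ten teeth.
# Each tooth has a dynamic capsule constrained to its own tiny base capsule, with individually tuned swing limits and almost no twist.
# All ten base capsules connect to one large dyn_body_05 capsule through constraints with free swing and twist, forming a hub for the necklace.
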PhysicsAsset /BRCosmetics/Weapons/FORT_Melee/Pickaxe_HangNine/Meshes/HangNine_Axe_Physics
bodies {
 Base_necklace_teeth_09_1 [{shapeType=Capsule, length=0.505, radius=0.505, shapeName=Base_necklace_teeth_09_1_capsule, rotation=<rot 0 -90 0>}]
 base_necklace_teeth_01_1 [{shapeType=Capsule, length=0.505, radius=0.505, shapeName=base_necklace_teeth_01_1_capsule, rotation=<rot 0 -90 0>}]
 base_necklace_teeth_02_1 [{shapeType=Capsule, length=0.505, radius=0.505, shapeName=base_necklace_teeth_02_1_capsule, rotation=<rot 0 -90 0>}]
 base_necklace_teeth_03_1 [{shapeType=Capsule, length=0.505, radius=0.505, shapeName=base_necklace_teeth_03_1_capsule, rotation=<rot 0 -90 0>}]
 base_necklace_teeth_04_1 [{shapeType=Capsule, length=0.505, radius=0.505, shapeName=base_necklace_teeth_04_1_capsule, rotation=<rot 0 -90 0>}]
 base_necklace_teeth_05_1 [{shapeType=Capsule, length=0.505, radius=0.505, shapeName=base_necklace_teeth_05_1_capsule, rotation=<rot 0 -90 0>}]
 base_necklace_teeth_06_1 [{shapeType=Capsule, length=0.505, radius=0.505, shapeName=base_necklace_teeth_06_1_capsule, rotation=<rot 0 -90 0>}]
 base_necklace_teeth_07_1 [{shapeType=Capsule, length=0.505, radius=0.505, shapeName=base_necklace_teeth_07_1_capsule, rotation=<rot 0 -90 0>}]
 base_necklace_teeth_08_1 [{shapeType=Capsule, length=0.505, radius=0.505, shapeName=base_necklace_teeth_08_1_capsule, rotation=<rot 0 -90 0>}]
 base_necklace_teeth_10_1 [{shapeType=Capsule, length=0.505, radius=0.505, shapeName=base_necklace_teeth_10_1_capsule, rotation=<rot 0 -90 0>}]
 dyn_body_05 [{shapeType=Capsule, length=11, radius=10.5, shapeName=dyn_body_05_capsule, center=<8.8921 0 0.118>, rotation=<rot 80 180 180>}]
 dyn_necklace_teeth_01_1 [{shapeType=Capsule, length=3, radius=1, shapeName=dyn_necklace_teeth_01_1_capsule, center=<3 0 0>, rotation=<rot 90 0 0>}]
 dyn_necklace_teeth_02_1 [{shapeType=Capsule, length=3, radius=1, shapeName=dyn_necklace_teeth_02_1_capsule, center=<3 0 0>, rotation=<rot 90 0 0>}]
 dyn_necklace_teeth_03_1 [{shapeType=Capsule, length=3, radius=1.1, shapeName=dyn_necklace_teeth_03_1_capsule, center=<3 0 0>, rotation=<rot 90 0 0>}]
 dyn_necklace_teeth_04_1 [{shapeType=Capsule, length=3, radius=0.8, shapeName=dyn_necklace_teeth_04_1_capsule, center=<3 0 0>, rotation=<rot 90 0 0>}]
 dyn_necklace_teeth_05_1 [{shapeType=Capsule, length=3, radius=0.8, shapeName=dyn_necklace_teeth_05_1_capsule, center=<3 0 0>, rotation=<rot 90 0 0>}]
 dyn_necklace_teeth_06_1 [{shapeType=Capsule, length=3, radius=1, shapeName=dyn_necklace_teeth_06_1_capsule, center=<3 0 0>, rotation=<rot 90 0 0>}]
 dyn_necklace_teeth_07_1 [{shapeType=Capsule, length=3, radius=1.1, shapeName=dyn_necklace_teeth_07_1_capsule, center=<3 0 0>, rotation=<rot 90 0 0>}]
 dyn_necklace_teeth_08_1 [{shapeType=Capsule, length=3, radius=1.1, shapeName=dyn_necklace_teeth_08_1_capsule, center=<3 0 0>, rotation=<rot 90 0 0>}]
 dyn_necklace_teeth_09_1 [{shapeType=Capsule, length=3, radius=1.1, shapeName=dyn_necklace_teeth_09_1_capsule, center=<3 0 0>, rotation=<rot 90 0 0>}]
 dyn_necklace_teeth_10_1 [{shapeType=Capsule, length=3, radius=1, shapeName=dyn_necklace_teeth_10_1_capsule, center=<3 0 0.9826>, rotation=<rot 90 0 0>}]
}
constraints [
 {bone1Name=dyn_necklace_teeth_01_1, bone2Name=base_necklace_teeth_01_1, swing1LimitDegrees=35, swing2LimitDegrees=70, twistLimitDegrees=1}
 {bone1Name=dyn_necklace_teeth_02_1, bone2Name=base_necklace_teeth_02_1, swing1LimitDegrees=40, swing2LimitDegrees=70, twistLimitDegrees=1}
 {bone1Name=dyn_necklace_teeth_03_1, bone2Name=base_necklace_teeth_03_1, swing1LimitDegrees=30, swing2LimitDegrees=45, twistLimitDegrees=1}
 {bone1Name=dyn_necklace_teeth_04_1, bone2Name=base_necklace_teeth_04_1, swing1LimitDegrees=30, swing2LimitDegrees=50, twistLimitDegrees=1}
 {bone1Name=dyn_necklace_teeth_05_1, bone2Name=base_necklace_teeth_05_1, swing1LimitDegrees=30, swing2LimitDegrees=45, twistLimitDegrees=1}
 {bone1Name=dyn_necklace_teeth_06_1, bone2Name=base_necklace_teeth_06_1, swing1LimitDegrees=30, swing2LimitDegrees=50, twistLimitDegrees=1}
 {bone1Name=dyn_necklace_teeth_07_1, bone2Name=base_necklace_teeth_07_1, swing1LimitDegrees=30, swing2LimitDegrees=45, twistLimitDegrees=1}
 {bone1Name=dyn_necklace_teeth_08_1, bone2Name=base_necklace_teeth_08_1, swing1LimitDegrees=30, swing2LimitDegrees=45, twistLimitDegrees=1}
 {bone1Name=dyn_necklace_teeth_09_1, bone2Name=Base_necklace_teeth_09_1, swing1LimitDegrees=35, swing2LimitDegrees=50, twistLimitDegrees=1}
 {bone1Name=dyn_necklace_teeth_10_1, bone2Name=base_necklace_teeth_10_1, swing1LimitDegrees=40, swing2LimitDegrees=70, twistLimitDegrees=1}
 {
  bone1Name=base_necklace_teeth_10_1
  bone2Name=dyn_body_05
  swing1LimitDegrees=45
  swing1Motion=Free
  swing2LimitDegrees=45
  swing2Motion=Free
  twistLimitDegrees=45
  twistMotion=Free
 }
 {
  bone1Name=Base_necklace_teeth_09_1
  bone2Name=dyn_body_05
  swing1LimitDegrees=45
  swing1Motion=Free
  swing2LimitDegrees=45
  swing2Motion=Free
  twistLimitDegrees=45
  twistMotion=Free
 }
 {
  bone1Name=base_necklace_teeth_08_1
  bone2Name=dyn_body_05
  swing1LimitDegrees=45
  swing1Motion=Free
  swing2LimitDegrees=45
  swing2Motion=Free
  twistLimitDegrees=45
  twistMotion=Free
 }
 {
  bone1Name=base_necklace_teeth_07_1
  bone2Name=dyn_body_05
  swing1LimitDegrees=45
  swing1Motion=Free
  swing2LimitDegrees=45
  swing2Motion=Free
  twistLimitDegrees=45
  twistMotion=Free
 }
 {
  bone1Name=base_necklace_teeth_06_1
  bone2Name=dyn_body_05
  swing1LimitDegrees=45
  swing1Motion=Free
  swing2LimitDegrees=45
  swing2Motion=Free
  twistLimitDegrees=45
  twistMotion=Free
 }
 {
  bone1Name=base_necklace_teeth_05_1
  bone2Name=dyn_body_05
  swing1LimitDegrees=45
  swing1Motion=Free
  swing2LimitDegrees=45
  swing2Motion=Free
  twistLimitDegrees=45
  twistMotion=Free
 }
 {
  bone1Name=base_necklace_teeth_04_1
  bone2Name=dyn_body_05
  swing1LimitDegrees=45
  swing1Motion=Free
  swing2LimitDegrees=45
  swing2Motion=Free
  twistLimitDegrees=45
  twistMotion=Free
 }
 {
  bone1Name=base_necklace_teeth_03_1
  bone2Name=dyn_body_05
  swing1LimitDegrees=45
  swing1Motion=Free
  swing2LimitDegrees=45
  swing2Motion=Free
  twistLimitDegrees=45
  twistMotion=Free
 }
 {
  bone1Name=base_necklace_teeth_02_1
  bone2Name=dyn_body_05
  swing1LimitDegrees=45
  swing1Motion=Free
  swing2LimitDegrees=45
  swing2Motion=Free
  twistLimitDegrees=45
  twistMotion=Free
 }
 {
  bone1Name=base_necklace_teeth_01_1
  bone2Name=dyn_body_05
  swing1LimitDegrees=45
  swing1Motion=Free
  swing2LimitDegrees=45
  swing2Motion=Free
  twistLimitDegrees=45
  twistMotion=Free
 }
]
